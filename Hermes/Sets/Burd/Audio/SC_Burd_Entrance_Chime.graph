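# Sound cue for the Burd set entrance chime, an ambient one-shot played at half volume.
# A single wave player feeds the chime sound straight to the output, using a dedicated attenuation asset and the Athena ambient sound class.
# Concurrency overrides allow one instance at a time with a retrigger delay, and virtualization is disabled.
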SoundCue /Hermes/Sets/Burd/Audio/SC_Burd_Entrance_Chime
root=WavePlayer_0
nodes [{id=WavePlayer_0, properties {soundWaveAssetPtr=&/Burd_Entrance_Chime_01}}]
properties {
 attenuationSettings=&/ATT_Burd_Entrance_Chime
 bOverrideConcurrency=true
 maxDistance=2100
 soundClassObject=&Fort_Ambient_Athena
 virtualizationMode=Disabled
 volumeMultiplier=0.5
 concurrencyOverrides {maxCount=1, retriggerTime=3}
}
